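# Mono one-shot MetaSound source for the Jethro BrightJuice shotgun, defaulting to an overheat sweetener wave with weapon-foley sound class and attenuation.
# OnPlay triggers a Wave Player whose pitch shift is the Pitch Shift input remapped through MapRange.
# Audio passes through a short delay to the output, while OnFinished comes straight from the player.
# Despite the "Randomized" name, the random pitch and volume inputs are unconnected.
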MetaSoundSource /JethroWeaponGameplay/Assets/Audio/BrightJuice/IP_MSS_Jethro_BrightJuice_Shotgun_OneShotSingleRandomized
inputs [
 Pitch Shift {type=Float}
 Random.Pitch.Base {type=Float}
 Random.Pitch.Spread {type=Float}
 Random.Volume.Base {type=Float}
 Random.Volume.Spread {type=Float}
 Sound {type=WaveAsset}
 UE.Source.OnPlay {type=Trigger}
]
interfaces [UE.OutputFormat.Mono:1.0, UE.Source.OneShot:1.0, UE.Source:1.1]
outputs [
 "UE.OutputFormat.Mono.Audio:0" {type=Audio}
 UE.Source.OneShot.OnFinished {type=Trigger}
]
pages [Default {
  input_defaults {
   Pitch Shift=0
   Random.Pitch.Base=0
   Random.Pitch.Spread=1
   Random.Volume.Base=1
   Random.Volume.Spread=0
   Sound=&/IP_Jethro_BrightJuice_Shotgun_Overheat_Sweetener_OS_01
   UE.Source.OnPlay=null
  }
  nodes [
   WavePlayer_0 = UE.Wave Player:Mono(
    Wave Asset=^Sound
    Play=^UE.Source.OnPlay
    Pitch Shift=MapRange.MapRange:Float(In=^Pitch Shift) {also {defaults {Out Range A=1, Out Range B=12}}}.Out Value
   )
   Delay_0 = UE.Delay:Audio(In=@WavePlayer_0.Out Mono) {also {defaults {Delay Time=0.1}}}
  ]
  links [{dst="@UE.OutputFormat.Mono.Audio:0", src=Delay_0.Out}, {dst="@UE.Source.OneShot.OnFinished", src=WavePlayer_0.On Finished}]
 }]
properties {
 assetImportData=&self:AssetImportData
 attenuationSettings=&AS_Fort_WeaponFoley
 maxDistance=2250
 sampleRate=48000
 soundClassObject=&Fort_WeaponFoley
 volume=1.7783
}
